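; Blueprint that switches the Fortnite public turntable scene between five lighting setups, each stored as a streaming sublevel.
; Switch Level, on authority only, prints the chosen enum name and switches on that string to unload the current sublevel and stream in the matching one.
; Each switch also sets a per-setup exposure compensation on the first post-process volume tagged ExposurePP.
; Begin play, tick, overlap, sun and older per-setup load events are empty.
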(event Collision|EventActorBeginOverlap
  (params (OtherActor :type Actor%20Object%20Reference)))

(event EventTick
  (params (DeltaSeconds :type Float%20%28single-precision%29)))

(event Custom|Load%20Front%20End%20Lighting%20Setup)

(event Custom|Load%20Br_Art%20Lighting%20Setup)

(event Custom|Load%20Neutral%20Lighting%20Setup)

(event Custom|Set%20False%20Sun)

(event Custom|Disable%20Real%20Sun%20System)

(event Custom|Activate%20Real%20Sun%20System)

(event EventBeginPlay)

(event Custom|Switch%20Level
  (params (Lighting%20SubLevel :type Lighting_Gbuffer_LevelSwitchEnum%20Enum :default "NewEnumerator0"))
  (bind _returnvalue (Utilities|String|EnumtoString Lighting%20SubLevel))
  (Networking|SwitchHasAuthority
    (:Authority
      (Development|PrintString _returnvalue)
      (switch string _returnvalue
        (:Neutral%20Lighting
          (CallFunction|HideLevel)
          (CallFunction|LoadLevel "GBufferLighting" "/Game/FN_PublicTurntable/TurnTable_gbuffer_LightingSublevels/GbufferLighting.GbufferLighting")
          (CallFunction|AdjustExposure 9.5))
        (:BR_Art%20Lighting
          (CallFunction|HideLevel)
          (CallFunction|LoadLevel "BRARTLightingInfo" "/Game/FN_PublicTurntable/TurnTable_gbuffer_LightingSublevels/BRARTLightingInfo.BRARTLightingInfo")
          (CallFunction|AdjustExposure 9.5))
        (:Front%20End%20Lighting
          (CallFunction|HideLevel)
          (CallFunction|LoadLevel "FELightingCopy" "/Game/FN_PublicTurntable/TurnTable_gbuffer_LightingSublevels/FELightingCopy.FELightingCopy")
          (CallFunction|AdjustExposure 10.1))
        (:TODM%20Lighting
          (CallFunction|HideLevel)
          (CallFunction|LoadLevel "BR_Art_Active_EditorOnly" "/Game/FN_PublicTurntable/TurnTable_gbuffer_LightingSublevels/BR_Art_Active_EditorOnly.BR_Art_Active_EditorOnly")
          (CallFunction|AdjustExposure 9.5))
        (:3%20Point%20Lighting
          (CallFunction|HideLevel)
          (CallFunction|LoadLevel "ThreePointLighting" "/Game/FN_PublicTurntable/TurnTable_gbuffer_LightingSublevels/ThreePointLighting.ThreePointLighting")
          (CallFunction|AdjustExposure 1e+01))))))

(event Custom|Load%20Level
  (params (PackageName :type Name) (Currently%20Loaded%20Level :type World%20Soft%20Object%20Reference))
  ;[1 1 1] Handle Sublevel Loading
  (bind _output_get (Variables|Default|SetCurrently%20Loaded%20Level Currently%20Loaded%20Level))
  (Game|LoadStreamLevel(byObjectReference) _output_get true))

(event Custom|Hide%20Level
  ;[1 1 1] Handle Sublevel Loading
  (Game|UnloadStreamLevel(byObjectReference) (Variables|Default|GetCurrently%20Loaded%20Level)))

(event Custom|Adjust%20Exposure
  (params (Exposure%20Compensation :type Float%20%28double-precision%29))
  (bind _outactors (Actor|GetAllActorswithTag "ExposurePP"))
  (bind _aspost_process_volume (Utilities|Casting|CastToPostProcessVolume (Utilities|Array|Get(acopy) _outactors)))
  (Utilities|Struct|SetmembersinPostProcessSettings (Variables|/Script/Engine.PostProcessVolume|GetSettings _aspost_process_volume) Exposure%20Compensation))

(fn ConstructionScript)
