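; TODM_Spawner is a Blueprint actor in the turntable support assets, named as a spawner.
; It holds a begin-overlap event that receives the other actor, a BeginPlay event and a construction script, all with empty bodies.
; No spawning logic appears in the graph.
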(event Collision|EventActorBeginOverlap
  (params (OtherActor :type Actor%20Object%20Reference)))

(event EventBeginPlay)

(fn ConstructionScript)
